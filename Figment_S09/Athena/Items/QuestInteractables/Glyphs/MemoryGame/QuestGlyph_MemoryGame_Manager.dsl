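; Quest glyph memory game manager: an indestructible switch that starts a Simon-style platform sequence game with a scoreboard.
; RepNotify variables drive switch state, game start and each platform's emote listening.
; Timers flash a random sequence that grows each round, then PlatformPlayed checks inputs, scoring the player completing each round.
; Mistakes reveal the correct platform and end the game, while an eligible quest stops new rounds at EndCount.
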(event Custom|StartGame
  ;[1 1 1] Start game
  (Variables|Default|SetCurrentScore 0)
  (Utilities|Array|Clear (Variables|Default|GetPlatformSequence))
  ;;[1 0.8843 0.4247] First index is 2 b/c Playback adds 1 to the sequence
  (for _ (range 2 (+ (Variables|Config|GetStartCount) 1))
    (Utilities|Array|Add (Variables|Default|GetPlatformSequence) (Math|Random|RandomIntegerinRange 0 (- (Utilities|Array|Length (Variables|Config|GetPlatforms)) 1))))
  (Utilities|FlowControl|Delay 2.0)
  (CallFunction|Playback))

(event Custom|Playback
  (Variables|Default|SetIsListeningForEmotes false)
  (Variables|Default|SetPlatformSequenceIndex 0)
  (Utilities|Array|Add (Variables|Default|GetPlatformSequence) (Math|Random|RandomIntegerinRange 0 (- (Utilities|Array|Length (Variables|Config|GetPlatforms)) 1)))
  (Networking|ForceNetUpdate)
  (Utilities|Time|SetTimerbyEvent (EventDispatchers|CreateEvent :!event Custom|LightUpNextPlatform) (/ (+ 1.0 (* (Utilities|Array|Length (Variables|Default|GetPlatformSequence)) 0.25)) (Math|Conversions|ToFloat(Integer) (Utilities|Array|Length (Variables|Default|GetPlatformSequence))))))

(event Custom|LightUpNextPlatform
  (bind _currentplatform (Variables|Default|GetCurrentPlatform))
  (Utilities|IsValid _currentplatform
    (:Is%20Valid
      (Variables|/Figment_S09/Athena/Items/QuestInteractables/Glyphs/MemoryGame/QuestGlyph_MemoryGame_Platform.QuestGlyph_MemoryGame_Platform_C|SetIsLit false _currentplatform))
    (:Is%20Not%20Valid
      _))
  (if (< (Variables|Default|GetPlatformSequenceIndex) (Utilities|Array|Length (Variables|Default|GetPlatformSequence)))
    (bind _output_get (Variables|Default|SetCurrentPlatform (Utilities|Array|Get(acopy) (Variables|Config|GetPlatforms) (Utilities|Array|Get(acopy) (Variables|Default|GetPlatformSequence) (Variables|Default|GetPlatformSequenceIndex)))))
    (Variables|/Figment_S09/Athena/Items/QuestInteractables/Glyphs/MemoryGame/QuestGlyph_MemoryGame_Platform.QuestGlyph_MemoryGame_Platform_C|SetIsLit true _output_get)
    (Math|Integer|IncrementInt (Variables|Default|GetPlatformSequenceIndex)
      (:%20%20
        (Utilities|Time|SetTimerbyEvent (EventDispatchers|CreateEvent :!event Custom|DimPlatform) (/ (+ 1.0 (* (Utilities|Array|Length (Variables|Default|GetPlatformSequence)) 0.25)) (Math|Conversions|ToFloat(Integer) (Utilities|Array|Length (Variables|Default|GetPlatformSequence)))))))
    (else
      (Variables|Default|SetIsListeningForEmotes true))))

(event Custom|DimPlatform
  (Variables|/Figment_S09/Athena/Items/QuestInteractables/Glyphs/MemoryGame/QuestGlyph_MemoryGame_Platform.QuestGlyph_MemoryGame_Platform_C|SetIsLit false (Variables|Default|GetCurrentPlatform))
  (Utilities|Time|SetTimerbyEvent (EventDispatchers|CreateEvent :!event Custom|LightUpNextPlatform) 0.1))

(event Interaction|EventBlueprintOnInteract
  (params (InteractingPawn :type Fort%20Pawn%20Object%20Reference) (InteractionBeingAttempted :type EInteractionBeingAttempted%20Enum :default "FirstInteraction"))
  (bind _returnvalue (Fortnite|GetFortPlayerControllerfromActor InteractingPawn))
  (bind _participatingplayers (Variables|Default|GetParticipatingPlayers))
  (bind _questdef (Variables|Default|GetQuestDef))
  ;[1 1 1] On Interact (player wants to start the game)
  (Class|BPScoreBoardPlace|Startup (Variables|Config|GetScoreBoard) InteractingPawn)
  (Utilities|Array|Clear _participatingplayers)
  (Utilities|Array|Add _participatingplayers _returnvalue)
  (Variables|Default|SetMemoryGameActive true)
  (bind (_questisvalid _questcompleted _objectivecompleted) (Default|AthenaCheckQuestandObjectiveCompletion _returnvalue _questdef (Variables|Default|GetQuestBackendObjectiveName)))
  (Variables|Default|SetUseEndCount (Math|Boolean|ANDBoolean _questisvalid (not _questcompleted) (not _objectivecompleted) (Quest|CanProgressQuestinCurrentZone (Quests|GetQuestManager _returnvalue "Invalid") _questdef))))

(event Custom|PlatformPlayed
  (params (Platform :type Quest%20Glyph%20Memory%20Game%20Platform%20Object%20Reference) (ContributingPlayer :type Fort%20Player%20Controller%20Object%20Reference))
  (if (Variables|Default|GetIsListeningForEmotes)
    (bind _output_get (Variables|Default|SetCurrentPlatform Platform))
    (Utilities|Array|AddUnique (Variables|Default|GetParticipatingPlayers) ContributingPlayer)
    (if (== (Utilities|Array|FindItem (Variables|Config|GetPlatforms) _output_get) (Utilities|Array|Get(aref) (Variables|Default|GetPlatformSequence) (Variables|Default|GetPlatformSequenceIndex)))
      (if (< (Variables|Default|GetPlatformSequenceIndex) (- (Utilities|Array|Length (Variables|Default|GetPlatformSequence)) 1))
        (Math|Integer|IncrementInt (Variables|Default|GetPlatformSequenceIndex))
        (else
          (bind _asfort_player_controller_athena (Utilities|Casting|CastToFortPlayerControllerAthena ContributingPlayer)
            (:then
              (Class|ScoreboardInterface|MoveScore (Variables|Config|GetScoreBoard) 1 (Utilities|Array|MakeArray _asfort_player_controller_athena)))
            (:CastFailed
              _))
          (if (not (and (Variables|Default|GetUseEndCount) (>= (Utilities|Array|Length (Variables|Default|GetPlatformSequence)) (Variables|Config|GetEndCount))))
            (CallFunction|Playback))))
      (else
        (Variables|Default|SetCorrectPlatform (Utilities|Array|Get(acopy) (Variables|Config|GetPlatforms) (Utilities|Array|Get(acopy) (Variables|Default|GetPlatformSequence) (Variables|Default|GetPlatformSequenceIndex))))
        (CallFunction|NotifyFailure)))))

(event Custom|NotifyFailure
  (Variables|Default|SetIsListeningForEmotes false)
  (Variables|/Figment_S09/Athena/Items/QuestInteractables/Glyphs/MemoryGame/QuestGlyph_MemoryGame_Platform.QuestGlyph_MemoryGame_Platform_C|SetIsLit true (Variables|Default|GetCorrectPlatform))
  (CallFunction|FailureCosmetics)
  (Utilities|Time|SetTimerbyEvent (EventDispatchers|CreateEvent :!event Custom|End%20Game) 2.0))

(event Custom|FailureCosmetics
  (Utilities|FlowControl|Delay 0.75)
  (Audio|PlaySoundatLocation "/Figment_S09/Athena/Sounds/Quests/Memory_Game/Memory_Game_Fail_Cue.Memory_Game_Fail_Cue" (Transformation|GetActorLocation)))

(event Custom|End%20Game
  (Variables|/Figment_S09/Athena/Items/QuestInteractables/Glyphs/MemoryGame/QuestGlyph_MemoryGame_Platform.QuestGlyph_MemoryGame_Platform_C|SetIsLit false (Variables|Default|GetCorrectPlatform))
  (Variables|Default|SetMemoryGameActive false))

(fn BlueprintCanInteract
  (params (InteractingPawn :type Fort%20Pawn%20Object%20Reference) (InteractionBeingAttempted :type EInteractionBeingAttempted%20Enum :default "FirstInteraction") (InteractionType :type TInteractionType%20Enum :default "IT_NoInteraction"))
  (return (not (Variables|Default|GetMemoryGameActive))))

(fn BlueprintGetInteractionString
  (params (InteractingPawn :type Fort%20Pawn%20Object%20Reference) (InteractionBeingAttempted :type EInteractionBeingAttempted%20Enum :default "FirstInteraction"))
  (return "NSLOCTEXT(\"[18090EF44BB80DE130E24886AF39E612]\", \"6EE6F95543BD036B8D7F8785E505E9D0\", \"Start Memory Game\")"))

(fn OnRep_IsListeningForEmotes
  (Variables|Default|SetPlatformSequenceIndex 0)
  (for _array_element (Variables|Config|GetPlatforms)
    (Variables|/Figment_S09/Athena/Items/QuestInteractables/Glyphs/MemoryGame/QuestGlyph_MemoryGame_Platform.QuestGlyph_MemoryGame_Platform_C|SetShouldListenForEmotes (Variables|Default|GetIsListeningForEmotes) _array_element)))

(fn OnRep_MemoryGameActive
  (if (Variables|Default|GetMemoryGameActive)
    (Audio|PlaySoundatLocation (Variables|Default|GetSwitchPullSound) (Transformation|GetActorLocation))
    (Transformation|SetRelativeRotation (Variables|Default|GetSwitchHandle) "40.000000, 0, 0" false false)
    (Variables|Interaction|SetbAllowInteract false)
    (CallFunction|StartGame)
    (else
      (Audio|PlaySoundatLocation (Variables|Default|GetSwitchResetSound) (Transformation|GetActorLocation))
      (Transformation|SetRelativeRotation (Variables|Default|GetSwitchHandle) "0, 0, 0" false false)
      (Variables|Interaction|SetbAllowInteract true)
      (Class|BPScoreBoardPlace|EndGame (Variables|Config|GetScoreBoard)))))

(fn ShouldDie
  (params (Damage :type Float%20%28single-precision%29) (EventInstigator :type Controller%20Object%20Reference) (DamageCauser :type Actor%20Object%20Reference) (bIsTest :type Boolean) (EffectContext :type Gameplay%20Effect%20Context%20Handle%20Structure%20%28by%20ref%29))
  (return false))

(fn ConstructionScript)
