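; A memory game quest platform that lights up, plays its note and reports player input to QuestGlyphMemoryGameManager.
; Toggling ShouldListenForEmotes binds or unbinds OnDamaged; a hit lights the platform, plays a hit sound, then after a short delay unlights it and sends the instigating controller to the manager.
; IsLit's RepNotify swaps the floor material and plays the note.
; Dance detection looks vestigial: overlap begin is empty and nothing here calls CheckForDancing.
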(event EventBeginPlay)

(event Custom|EnableListenForDancing
  (EventDispatchers|BindEventtoOnDamaged (delegate Custom|PlatformHit)))

(event Custom|DisableListenForDancing
  (EventDispatchers|UnbindEventfromOnDamaged (delegate Custom|PlatformHit))
  (Utilities|Time|ClearandInvalidateTimerbyHandle (Variables|Default|GetDanceCheckTimer))
  (Utilities|Array|Clear (Variables|Default|GetDanceCandidates)))

(event Custom|PlatformHit
  (params (DamagedActor :type Actor%20Object%20Reference) (Damage :type Float%20%28single-precision%29) (InstigatedBy :type Controller%20Object%20Reference) (DamageCauser :type Actor%20Object%20Reference) (HitLocation :type Vector) (FHitComponent :type Primitive%20Component%20Object%20Reference) (BoneName :type Name) (Momentum :type Vector))
  (Variables|Default|SetIsLit true)
  (CallFunction|PlayHitSound)
  (Utilities|FlowControl|Delay 0.25)
  (bind _asfort_player_controller (Utilities|Casting|CastToFortPlayerController InstigatedBy))
  (CallFunction|PlatformTriggered _asfort_player_controller))

(event Custom|PawnBeginsOverlap
  (params (OverlappedComponent :type Primitive%20Component%20Object%20Reference) (OtherActor :type Actor%20Object%20Reference) (OtherComp :type Primitive%20Component%20Object%20Reference) (OtherBodyIndex :type Integer) (bFromSweep :type Boolean) (SweepResult :type Hit%20Result%20Structure%20%28by%20ref%29)))

(event Custom|PawnEndsOverlap
  (params (OverlappedComponent :type Primitive%20Component%20Object%20Reference) (OtherActor :type Actor%20Object%20Reference) (OtherComp :type Primitive%20Component%20Object%20Reference) (OtherBodyIndex :type Integer))
  (bind _dancecandidates (Variables|Default|GetDanceCandidates))
  (Utilities|Array|RemoveItem _dancecandidates OtherActor)
  (if (<= (Utilities|Array|Length _dancecandidates) 0)
    (Variables|Default|SetIsLit false)
    (Class|QuestGlyphMemoryGameManager|PlatformPlayed (Variables|Default|GetMemoryGameManager) self (stale Custom|PlatformTriggered Player))))

(event Custom|PlatformTriggered
  (params (Player :type Fort%20Player%20Controller%20Object%20Reference))
  (Variables|Default|SetIsLit false)
  (Class|QuestGlyphMemoryGameManager|PlatformPlayed (Variables|Default|GetMemoryGameManager) self Player))

(event Custom|CheckForDancing
  (Utilities|Array|Clear (Variables|Default|GetContributingPlayers))
  (Variables|Default|SetSomeoneIsDancing false)
  (bind _array_element (Utilities|Array|ForEachLoopwithBreak (Variables|Default|GetDanceCandidates))
    (:LoopBody
      (bind _asfort_player_controller (Utilities|Casting|CastToFortPlayerController (Pawn|GetController (Variables|/Script/Engine.Actor|GetInstigator _array_element))))
      (Utilities|Array|AddUnique (Variables|Default|GetContributingPlayers) _asfort_player_controller)
      (Variables|Default|SetSomeoneIsDancing true)
      (if (and (Variables|Default|GetShouldListenForEmotes) (not (Variables|Default|GetIsLit)))
        (Variables|Default|SetIsLit true)))
    (:Completed
      (if (and (not (Variables|Default|GetSomeoneIsDancing)) (Variables|Default|GetIsLit))
        (Variables|Default|SetIsLit false)
        (if (Variables|Default|GetShouldListenForEmotes)
          _)))))

(event Custom|Play%20Hit%20Sound
  (Audio|PlaySoundatLocation "/Figment_S09/Athena/Sounds/Quests/Memory_Game/Memory_Game_Note_Hit_Cue.Memory_Game_Note_Hit_Cue" (Transformation|GetActorLocation)))

(event Custom|TurnOffLight
  (Variables|Default|SetIsLit false))

(fn OnRep_IsLit
  (bind _dancefloor (Variables|Default|GetDanceFloor))
  (if (Variables|Default|GetIsLit)
    (Rendering|Material|SetMaterial _dancefloor 0 (Variables|Default|GetPlatFormColor))
    (Audio|PlaySoundatLocation (Variables|Default|GetPlatformNote) (Transformation|GetActorLocation))
    (else
      (Rendering|Material|SetMaterial _dancefloor 0 (Variables|Default|GetPlatFormColor_Dark)))))

(fn OnRep_ShouldListenForEmotes
  (Variables|Default|SetIsLit false)
  (if (Variables|Default|GetShouldListenForEmotes)
    (CallFunction|EnableListenforDancing)
    (else
      (CallFunction|DisableListenforDancing))))

(fn ConstructionScript
  (EventDispatchers|UnbindallEventsfromOnDamaged)
  (Rendering|Material|SetMaterial (Variables|Default|GetDanceFloor) 0 (Variables|Default|GetPlatFormColor_Dark)))
